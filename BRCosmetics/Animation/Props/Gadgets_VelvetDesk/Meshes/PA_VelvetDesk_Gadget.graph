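# Physics asset for the Velvet Desk gadget prop skeletal mesh in the BRCosmetics animation props.
# It defines a single body on the main bone, using a zero-length capsule of radius 95 that acts as a sphere.
# It has no constraints, so the prop simulates and collides as one rigid shape rather than articulated parts.
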PhysicsAsset /BRCosmetics/Animation/Props/Gadgets_VelvetDesk/Meshes/PA_VelvetDesk_Gadget
bodies {main [{shapeType=Capsule, length=0, radius=95, rotation=<rot 0 -90 0>}]}
constraints []
